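# Physics asset for the F_MED_Slither CAU face accessory mesh, giving capsule bodies and joint limits to its hair and headband bones.
# Two hair chains and a headband root hang off the head, with the headband branching into left and right strands.
# Chain links have tight twist and angular limits, with hair1–3 restricted to one swing axis.
# Head, neck and upper spine capsules join through free angular constraints, as do the hairA and headband roots.
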PhysicsAsset /BRCosmetics/Characters/Player/Female/Medium/Bodies/F_MED_Slither/Meshes/Parts/CAU/F_MED_Slither_CAU_FaceAcc_Physics
bodies {
 hair1 [{shapeType=Capsule, length=1.75, radius=1.75, shapeName=hairA2_capsule, center=<2 -1.8 0>, rotation=<rot 90 0 0>}]
 hair2 [{shapeType=Capsule, length=1.75, radius=1.75, shapeName=hairA2_capsule, center=<2 -1.8 0>, rotation=<rot 90 0 0>}]
 hair3 [{shapeType=Capsule, length=1.75, radius=1.75, shapeName=hairA2_capsule, center=<2 -1.8 0>, rotation=<rot 90 0 0>}]
 hairA1 [{shapeType=Capsule, length=1, radius=2, shapeName=hairA1_capsule, rotation=<rot 90 0 0>}]
 hairA2 [{shapeType=Capsule, length=0, radius=3.5, shapeName=hairA2_capsule, center=<2 0 0>, rotation=<rot 90 0 0>}]
 hairA3 [{shapeType=Capsule, length=0.5, radius=4, shapeName=hairA3_capsule, center=<2 0 0>, rotation=<rot 90 0 0>}]
 hairA4 [{shapeType=Capsule, length=1.5, radius=4.75, shapeName=hairA4_capsule, center=<2 0 0>, rotation=<rot 90 0 0>}]
 hairA5 [{shapeType=Capsule, length=1, radius=3.75, shapeName=hairA5_capsule, center=<3.765 -0.0001 0>, rotation=<rot 90 0 0>}]
 head [{shapeType=Capsule, length=0, radius=8.3681, shapeName=head_capsule, center=<5.1648 0.504 0.0002>}]
 headbandLf1 [{shapeType=Capsule, length=1, radius=1.5, shapeName=headbandLf1_capsule, center=<2 0 0>, rotation=<rot 90 0 0>}]
 headbandLf2 [{shapeType=Capsule, length=1, radius=1.5, shapeName=headbandLf2_capsule, center=<2 0 0>, rotation=<rot 90 0 0>}]
 headbandMain [{shapeType=Capsule, length=1, radius=2, shapeName=headbandMain_capsule, rotation=<rot 90 0 0>}]
 headbandRt1 [{shapeType=Capsule, length=1, radius=1.5, shapeName=headbandRt1_capsule, center=<2 0 0>, rotation=<rot 90 0 0>}]
 headbandRt2 [{shapeType=Capsule, length=0.75, radius=1.75, shapeName=headbandRt2_capsule, center=<2 0 0>, rotation=<rot 90 0 0>}]
 headbandRt3 [{shapeType=Capsule, length=2.75, radius=2, shapeName=headbandRt3_capsule, center=<2 0 0>, rotation=<rot 90 0 0>}]
 neck_01 [{shapeType=Capsule, length=8.25, radius=4.005, shapeName=neck_01_capsule, center=<3.3939 0 0>, rotation=<rot 90 0.844 0.8441>}]
 neck_02 [{shapeType=Capsule, length=2, radius=4.005, shapeName=neck_02_capsule, center=<2.5525 0 0>, rotation=<rot 90 0 0>}]
 spine_05 [{shapeType=Capsule, length=7.505, radius=10.005, shapeName=spine_05_capsule, center=<1.7907 0.9026 0>, rotation=<rot 0 -89.9999 0>}]
}
constraints [
 {bone1Name=headbandLf2, bone2Name=headbandLf1, swing1LimitDegrees=45, swing2LimitDegrees=5, twistLimitDegrees=5}
 {bone1Name=headbandRt3, bone2Name=headbandRt2, swing1LimitDegrees=45, swing2LimitDegrees=45, twistLimitDegrees=5}
 {bone1Name=headbandRt2, bone2Name=headbandRt1, swing1LimitDegrees=45, swing2LimitDegrees=45, twistLimitDegrees=5}
 {bone1Name=headbandRt1, bone2Name=headbandMain, swing1LimitDegrees=45, swing2LimitDegrees=45, twistLimitDegrees=5}
 {bone1Name=headbandLf1, bone2Name=headbandMain, swing1LimitDegrees=45, swing2LimitDegrees=5, twistLimitDegrees=5}
 {bone1Name=hairA5, bone2Name=hairA4, swing1LimitDegrees=10, swing2LimitDegrees=45, twistLimitDegrees=5}
 {bone1Name=hairA4, bone2Name=hairA3, swing1LimitDegrees=10, swing2LimitDegrees=45, twistLimitDegrees=5}
 {bone1Name=hairA3, bone2Name=hairA2, swing1LimitDegrees=10, swing2LimitDegrees=45, twistLimitDegrees=5}
 {bone1Name=hairA2, bone2Name=hairA1, swing1LimitDegrees=10, swing2LimitDegrees=45, twistLimitDegrees=5}
 {
  bone1Name=hairA1
  bone2Name=head
  swing1LimitDegrees=45
  swing1Motion=Free
  swing2LimitDegrees=45
  swing2Motion=Free
  twistLimitDegrees=45
  twistMotion=Free
 }
 {
  bone1Name=headbandMain
  bone2Name=head
  swing1LimitDegrees=45
  swing1Motion=Free
  swing2LimitDegrees=45
  swing2Motion=Free
  twistLimitDegrees=45
  twistMotion=Free
 }
 {
  bone1Name=head
  bone2Name=neck_02
  swing1LimitDegrees=45
  swing1Motion=Free
  swing2LimitDegrees=45
  swing2Motion=Free
  twistLimitDegrees=45
  twistMotion=Free
 }
 {
  bone1Name=neck_02
  bone2Name=neck_01
  swing1LimitDegrees=45
  swing1Motion=Free
  swing2LimitDegrees=45
  swing2Motion=Free
  twistLimitDegrees=45
  twistMotion=Free
 }
 {
  bone1Name=neck_01
  bone2Name=spine_05
  swing1LimitDegrees=45
  swing1Motion=Free
  swing2LimitDegrees=45
  swing2Motion=Free
  twistLimitDegrees=45
  twistMotion=Free
 }
 {bone1Name=hair1, bone2Name=head, swing1LimitDegrees=0, swing2LimitDegrees=12, twistLimitDegrees=0}
 {bone1Name=hair2, bone2Name=hair1, swing1LimitDegrees=0, swing2LimitDegrees=6, twistLimitDegrees=0}
 {bone1Name=hair3, bone2Name=hair2, swing1LimitDegrees=0, swing2LimitDegrees=12, twistLimitDegrees=0}
]
