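# Animation Blueprint for the Slither female medium body's CAU face accessory, adding hair physics and a crouch-sprint pose adjustment.
# It copies the attached parent's pose, then runs RigidBody simulation relative to the head bone with damped component motion.
# Additive Modify Bone nodes on hairA1–A5 then offset the hair.
# Their shared alpha is the crouch-sprinting bool converted to float each update, with interpolation smoothing the blend.
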AnimBlueprint /BRCosmetics/Characters/Player/Female/Medium/Bodies/F_MED_Slither/Meshes/Parts/CAU/F_MED_Slither_CAU_FaceAcc_AnimBP
target_skeleton=&/F_MED_Slither_CAU_FaceAcc_Skeleton
graphs [
 AnimGraph {nodes [
   n1 = RigidBody(ComponentPose=Local To Component(LocalPose=Copy Pose From Mesh() {bUseAttachedParent=true})) {
    bClampLinearTranslationLimitToRefPose=true
    bOverrideWorldGravity=true
    overridePhysicsAsset=&/F_MED_Slither_CAU_FaceAcc_Physics
    simulationSpace=BaseBoneSpace
    alphaBoolBlend {blendOption=Linear}
    baseBoneRef {boneName=head}
    componentAppliedLinearAccClamp=<3000 3000 3000>
    componentLinearAccScale=<0.4 0.3 0.3>
    componentLinearVelScale=<0.4 0.3 0.3>
    overrideWorldGravity {z=-980}
   }
   n2 = Get crouchSprinting() -> (crouchSprinting:double)
   n3 = "Transform (Modify) Bone - Bone: hairA4"(ComponentPose=@n1, Alpha=@n2, Translation="0,0,0", Rotation="0,0,0", Scale="1,1,1") {
    rotationMode=BMM_Additive
    scaleMode=BMM_Additive
    translationMode=BMM_Additive
    alphaBoolBlend {blendOption=Linear}
    alphaScaleBiasClamp {bInterpResult=true, interpSpeedDecreasing=1, interpSpeedIncreasing=1}
    boneToModify {boneName=hairA4}
   }
   n4 = "Transform (Modify) Bone - Bone: hairA3"(ComponentPose=@n3, Alpha=@n2, Translation="0,-2,0", Rotation="0,0,-15", Scale="1,1,1") {
    rotationMode=BMM_Additive
    scaleMode=BMM_Additive
    translationMode=BMM_Additive
    alphaBoolBlend {blendOption=Linear}
    alphaScaleBiasClamp {bInterpResult=true, interpSpeedDecreasing=1, interpSpeedIncreasing=1}
    boneToModify {boneName=hairA3}
    rotation {roll=-15}
    translation {y=-2}
   }
   n5 = "Transform (Modify) Bone - Bone: hairA2"(ComponentPose=@n4, Alpha=@n2, Translation="0,0,0", Rotation="0,0,0", Scale="1,1,1") {
    rotationMode=BMM_Additive
    scaleMode=BMM_Additive
    translationMode=BMM_Additive
    alphaBoolBlend {blendOption=Linear}
    alphaScaleBiasClamp {bInterpResult=true, interpSpeedDecreasing=1, interpSpeedIncreasing=1}
    boneToModify {boneName=hairA2}
   }
   n6 = "Transform (Modify) Bone - Bone: hairA1"(ComponentPose=@n5, Alpha=@n2, Translation="0,-5,0", Rotation="0,0,-15", Scale="1,1,1") {
    rotationMode=BMM_Additive
    scaleMode=BMM_Additive
    translationMode=BMM_Additive
    alphaBoolBlend {blendOption=Linear}
    alphaScaleBiasClamp {bInterpResult=true, interpSpeedDecreasing=1, interpSpeedIncreasing=1}
    boneToModify {boneName=hairA1}
    rotation {roll=-15}
    translation {y=-5}
   }
   n7 = "Transform (Modify) Bone - Bone: hairA5"(ComponentPose=@n6, Alpha=@n2, Translation="0,0,0", Rotation="0,0,-15", Scale="1,1,1") {
    rotationMode=BMM_Additive
    scaleMode=BMM_Additive
    translationMode=BMM_Additive
    alphaBoolBlend {blendOption=Linear}
    alphaScaleBiasClamp {bInterpResult=true, interpSpeedDecreasing=1, interpSpeedIncreasing=1}
    boneToModify {boneName=hairA5}
    rotation {roll=-15}
   }
   Output Pose(Result=Component To Local(ComponentPose=@n7)) {layerGroup=None}
  ]}
 EventGraph {nodes [
   Event BlueprintUpdateAnimation()
   > Set crouchSprinting(crouchSprinting:double="To Float (Boolean)"(InBool=Get bIsCrouchSprinting())) -> (Output_Get:double)
  ]}
]
variables [crouchSprinting {type=number}]
